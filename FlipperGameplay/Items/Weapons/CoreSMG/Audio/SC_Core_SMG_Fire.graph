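# Core SMG fire SoundCue that layers close, medium and distant gunshot sounds.
# Each layer randomly picks a wave, with medium favoring one variant, then applies its own distance attenuation and a sound class override.
# Close and medium are mixed first, then blended with distant; close uses a 3P fire class, while the others share the distant class.
# A per-owner concurrency override caps overlapping shots, stopping the oldest.
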SoundCue /FlipperGameplay/Items/Weapons/CoreSMG/Audio/SC_Core_SMG_Fire
root=Mixer_2
nodes [
 {id=Attenuation_3, children [Random_0], properties {attenuationSettings=&Fort_Weapon_Projectile_Close_Athena_Att}}
 {
  id=Attenuation_4
  children [Modulator_4]
  properties {attenuationSettings=&Fort_Weapon_Projectile_Distant_Mono_Athena_Att}
 }
 {
  id=Attenuation_5
  children [Modulator_1]
  properties {attenuationSettings=&Fort_Weapon_Projectile_Medium_Mono_Athena_Att}
 }
 {id=Mixer_2, children [Modulator_3,SoundClass_4], properties {inputVolume [1,0.6]}}
 {id=Mixer_3, children [SoundClass_3,SoundClass_5], properties {inputVolume [0.65,1]}}
 {id=Modulator_1, children [Random_4], properties {volumeMax=1, volumeMin=0.9}}
 {id=Modulator_3, children [Mixer_3], properties {pitchMax=1, pitchMin=1}}
 {id=Modulator_4, children [Random_5]}
 {
  id=Random_0
  children [WavePlayer_0,WavePlayer_1,WavePlayer_2,WavePlayer_3,WavePlayer_4,WavePlayer_5]
  properties {bShouldExcludeFromBranchCulling=true, weights [1,1,1,1,1,1]}
 }
 {
  id=Random_4
  children [WavePlayer_6, WavePlayer_7, WavePlayer_10, WavePlayer_12, WavePlayer_13, WavePlayer_14]
  properties {bShouldExcludeFromBranchCulling=true, weights [1,1,1,4,1,1]}
 }
 {
  id=Random_5
  children [WavePlayer_15, WavePlayer_16, WavePlayer_17, WavePlayer_18, WavePlayer_19, WavePlayer_20, WavePlayer_21]
  properties {bShouldExcludeFromBranchCulling=true, weights [1,1,1,1,1,1,1]}
 }
 {
  id=SoundClass_3
  children [Attenuation_3]
  properties {soundClassOverride=&Fort_WeaponRanged_Fire_3P_PSM}
 }
 {
  id=SoundClass_4
  children [Attenuation_4]
  properties {soundClassOverride=&Fort_WeaponRanged_Fire_Distant}
 }
 {
  id=SoundClass_5
  children [Attenuation_5]
  properties {soundClassOverride=&Fort_WeaponRanged_Fire_Distant}
 }
 {id=WavePlayer_0, properties {soundWaveAssetPtr=&/Core_SMG_Fire_04}}
 {id=WavePlayer_1, properties {soundWaveAssetPtr=&/Core_SMG_Fire_01}}
 {id=WavePlayer_10, properties {soundWaveAssetPtr=&/Core_SMG_Fire_Medium_03}}
 {id=WavePlayer_12, properties {soundWaveAssetPtr=&/Core_SMG_Fire_Medium_06}}
 {id=WavePlayer_13, properties {soundWaveAssetPtr=&/Core_SMG_Fire_Medium_08}}
 {id=WavePlayer_14, properties {soundWaveAssetPtr=&/Core_SMG_Fire_Medium_09}}
 {id=WavePlayer_15, properties {soundWaveAssetPtr=&/Core_SMG_Fire_Distant_01}}
 {id=WavePlayer_16, properties {soundWaveAssetPtr=&/Core_SMG_Fire_Distant_02}}
 {id=WavePlayer_17, properties {soundWaveAssetPtr=&/Core_SMG_Fire_Distant_03}}
 {id=WavePlayer_18, properties {soundWaveAssetPtr=&/Core_SMG_Fire_Distant_04}}
 {id=WavePlayer_19, properties {soundWaveAssetPtr=&/Core_SMG_Fire_Distant_05}}
 {id=WavePlayer_2, properties {soundWaveAssetPtr=&/Core_SMG_Fire_02}}
 {id=WavePlayer_20, properties {soundWaveAssetPtr=&/Core_SMG_Fire_Distant_06}}
 {id=WavePlayer_21, properties {soundWaveAssetPtr=&/Core_SMG_Fire_Distant_07}}
 {id=WavePlayer_3, properties {soundWaveAssetPtr=&/Core_SMG_Fire_03}}
 {id=WavePlayer_4, properties {soundWaveAssetPtr=&/Core_SMG_Fire_05}}
 {id=WavePlayer_5, properties {soundWaveAssetPtr=&/Core_SMG_Fire_06}}
 {id=WavePlayer_6, properties {soundWaveAssetPtr=&/Core_SMG_Fire_Medium_04}}
 {id=WavePlayer_7, properties {soundWaveAssetPtr=&/Core_SMG_Fire_Medium_07}}
]
properties {
 bDebug=true
 bOverrideConcurrency=true
 maxDistance=27500
 volumeMultiplier=0.6
 concurrencyOverrides {bLimitToOwner=true, maxCount=2, resolutionRule=StopOldest}
}
