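# First-person fire SoundCue for the Core SMG, with owner-limited, stop-oldest concurrency.
# A quality node picks one of two parallel mixes, the last two levels using fewer random variations.
# Each mix layers random shots and pitched-down shell impacts, volume-modulated by TimeSinceStarted, with a LowAmmo-branched warning under its own sound class.
# The FirstShot branches are empty and the delayed mechanical layer is mixed at zero volume.
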SoundCue /FlipperGameplay/Items/Weapons/CoreSMG/Audio/SC_Core_SMG_Fire_1P
root=QualityLevel_0
nodes [
 {id=Branch_0, children [null,null,null], properties {boolParameterName=FirstShot}}
 {id=Branch_1, children [null,null,null], properties {boolParameterName=FirstShot}}
 {id=Branch_3, children [Modulator_5,null,null], properties {boolParameterName=LowAmmo}}
 {id=Branch_5, children [Modulator_9,null,null], properties {boolParameterName=LowAmmo}}
 {id=Delay_1, children [WavePlayer_23]}
 {id=Delay_2, children [WavePlayer_13]}
 {id=Mixer_0, children [Modulator_0,Modulator_2,Delay_1], properties {inputVolume [1,0.3,0]}}
 {id=Mixer_1, children [SoundClass_0,SoundClass_1], properties {inputVolume [1.2,1]}}
 {id=Mixer_2, children [Branch_1, ModulatorContinuous_0], properties {inputVolume [1,1]}}
 {id=Mixer_3, children [Modulator_1,Modulator_6,Delay_2], properties {inputVolume [1,0.3,0]}}
 {id=Mixer_4, children [SoundClass_2,SoundClass_3], properties {inputVolume [1.2,1]}}
 {id=Mixer_5, children [Branch_0, ModulatorContinuous_1], properties {inputVolume [1,1]}}
 {
  id=ModulatorContinuous_0
  children [Modulator_3]
  properties {volumeModulationParams {default=0, maxOutput=0.6, minOutput=1, parameterName=TimeSinceStarted}}
 }
 {
  id=ModulatorContinuous_1
  children [Modulator_7]
  properties {volumeModulationParams {default=0, maxOutput=0.6, minOutput=1, parameterName=TimeSinceStarted}}
 }
 {id=Modulator_0, children [Random_0], properties {pitchMax=1, pitchMin=1}}
 {id=Modulator_1, children [Random_1], properties {pitchMax=1, pitchMin=1}}
 {id=Modulator_2, children [Random_5], properties {pitchMax=0.95, pitchMin=0.85}}
 {id=Modulator_3, children [Mixer_0], properties {pitchMax=1, pitchMin=1}}
 {id=Modulator_4, children [Branch_3], properties {pitchMax=1, pitchMin=1, volumeMax=1.15, volumeMin=1}}
 {id=Modulator_5, children [WavePlayer_6], properties {pitchMax=1.1, pitchMin=1.1, volumeMax=1.2, volumeMin=1.2}}
 {id=Modulator_6, children [Random_3], properties {pitchMax=0.95, pitchMin=0.85}}
 {id=Modulator_7, children [Mixer_3], properties {pitchMax=1, pitchMin=1}}
 {id=Modulator_8, children [Branch_5], properties {pitchMax=1, pitchMin=1, volumeMax=1.15, volumeMin=1}}
 {id=Modulator_9, children [WavePlayer_25], properties {pitchMax=1.1, pitchMin=1.1, volumeMax=1.2, volumeMin=1.2}}
 {id=QualityLevel_0, children [Mixer_1,Mixer_1,Mixer_4,Mixer_4]}
 {id=Random_0, children [WavePlayer_0,WavePlayer_1,WavePlayer_2,WavePlayer_3,WavePlayer_4,WavePlayer_5], properties {weights [1,1,1,1,1,1]}}
 {id=Random_1, children [WavePlayer_7,WavePlayer_8,WavePlayer_9], properties {weights [1,1,1]}}
 {id=Random_3, children [WavePlayer_14, WavePlayer_15, WavePlayer_16], properties {weights [1,1,1]}}
 {
  id=Random_5
  children [WavePlayer_27, WavePlayer_28, WavePlayer_29, WavePlayer_30, WavePlayer_31, WavePlayer_32, WavePlayer_33]
  properties {weights [1,1,1,1,1,1,1]}
 }
 {
  id=SoundClass_0
  children [Modulator_4]
  properties {soundClassOverride=/Game/Sounds/FortSoundClassesAndModes/Fort_WeaponRanged_LowAmmo_Critical_PSM}
 }
 {id=SoundClass_1, children [Mixer_2], properties {soundClassOverride=&Fort_WeaponRanged_Fire_1P_PSM}}
 {
  id=SoundClass_2
  children [Modulator_8]
  properties {soundClassOverride=&Fort_WeaponRanged_LowAmmo_Critical_PSM}
 }
 {id=SoundClass_3, children [Mixer_5], properties {soundClassOverride=&Fort_WeaponRanged_Fire_1P_PSM}}
 {id=WavePlayer_0, properties {soundWaveAssetPtr=&/Core_SMG_Fire_01}}
 {id=WavePlayer_1, properties {soundWaveAssetPtr=&/Core_SMG_Fire_02}}
 {id=WavePlayer_13, properties {soundWaveAssetPtr=&/Core_SMG_FireMech1P_01}}
 {id=WavePlayer_14, properties {soundWaveAssetPtr=&/Core_SMG_ShellImpact_04}}
 {id=WavePlayer_15, properties {soundWaveAssetPtr=&/Core_SMG_ShellImpact_01}}
 {id=WavePlayer_16, properties {soundWaveAssetPtr=&/Core_SMG_ShellImpact_02}}
 {id=WavePlayer_2, properties {soundWaveAssetPtr=&/Core_SMG_Fire_03}}
 {id=WavePlayer_23, properties {soundWaveAssetPtr=&Core_SMG_FireMech1P_01}}
 {id=WavePlayer_25, properties {soundWaveAssetPtr=&/Core_SMG_LowAmmo_02}}
 {id=WavePlayer_27, properties {soundWaveAssetPtr=&Core_SMG_ShellImpact_04}}
 {id=WavePlayer_28, properties {soundWaveAssetPtr=&Core_SMG_ShellImpact_01}}
 {id=WavePlayer_29, properties {soundWaveAssetPtr=&Core_SMG_ShellImpact_02}}
 {id=WavePlayer_3, properties {soundWaveAssetPtr=&/Core_SMG_Fire_04}}
 {id=WavePlayer_30, properties {soundWaveAssetPtr=&/Core_SMG_ShellImpact_03}}
 {id=WavePlayer_31, properties {soundWaveAssetPtr=&/Core_SMG_ShellImpact_05}}
 {id=WavePlayer_32, properties {soundWaveAssetPtr=&/Core_SMG_ShellImpact_06}}
 {id=WavePlayer_33, properties {soundWaveAssetPtr=&/Core_SMG_ShellImpact_07}}
 {id=WavePlayer_4, properties {soundWaveAssetPtr=&/Core_SMG_Fire_05}}
 {id=WavePlayer_5, properties {soundWaveAssetPtr=&/Core_SMG_Fire_06}}
 {id=WavePlayer_6, properties {soundWaveAssetPtr=&Core_SMG_LowAmmo_02}}
 {id=WavePlayer_7, properties {soundWaveAssetPtr=&Core_SMG_Fire_01}}
 {id=WavePlayer_8, properties {soundWaveAssetPtr=&Core_SMG_Fire_02}}
 {id=WavePlayer_9, properties {soundWaveAssetPtr=&Core_SMG_Fire_03}}
]
properties {
 attenuationSettings=&Fort_Weapon_Projectile_FirstPerson_Athena_Att
 bDebug=true
 maxDistance=2097152
 soundClassObject=&Fort_WeaponRanged_Fire_1P_PSM
 volumeMultiplier=0.8
 attenuationOverrides {manualReverbSendLevel=-1.0165, reverbSendMethod=Manual}
 concurrencyOverrides {bLimitToOwner=true, maxCount=2, resolutionRule=StopOldest, volumeScale=0.9}
 concurrencySet [&Weapon_Gunfire_FirstPerson_Conc]
}
