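# First-person sound cue for the Core SMG's fire stop, the sound that closes out gunfire.
# A single wave player reuses the MilitarySMG close fire-stop sample 03 and feeds a Modulator that slightly randomizes volume on each play.
# The cue lowers overall volume and uses first-person weapon attenuation, ranged-fire sound class and gunfire concurrency.
# The cue's debug flag is switched on.
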SoundCue /FlipperGameplay/Items/Weapons/CoreSMG/Audio/SC_Core_SMG_Fire_Stop_1P
root=Modulator_0
nodes [
 {id=Modulator_0, children [WavePlayer_5], properties {volumeMax=0.85, volumeMin=0.8}}
 {id=WavePlayer_5, properties {soundWaveAssetPtr=/Game/Sounds/Fort_Weapons/AutoRifle/MilitarySMG/MilitarySMG_Close_FireStop_03}}
]
properties {
 attenuationSettings=&Fort_Weapon_Projectile_FirstPerson_Athena_Att
 bDebug=true
 maxDistance=2097152
 soundClassObject=&Fort_WeaponRanged_Fire_1P_PSM
 volumeMultiplier=0.7
 concurrencySet [&Weapon_Gunfire_FirstPerson_Conc]
}
