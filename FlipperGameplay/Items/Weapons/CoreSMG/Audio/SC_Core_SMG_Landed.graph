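# Sound cue for the Core SMG landing impact, likely when the weapon drops to the ground.
# A single drop-impact wave plays through a Modulator node at slightly reduced cue volume, in the loot-collect sound class.
# It overrides attenuation with a short falloff and limits concurrency to a few instances, blocking new ones once full.
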SoundCue /FlipperGameplay/Items/Weapons/CoreSMG/Audio/SC_Core_SMG_Landed
root=Modulator_0
nodes [
 {id=Modulator_0, children [WavePlayer_1]}
 {id=WavePlayer_1, properties {soundWaveAssetPtr=&/Core_SMG_DropImpactNew}}
]
properties {
 attenuationSettings=&Foley_Attenuation_Basic_Att
 bOverrideAttenuation=true
 bOverrideConcurrency=true
 maxDistance=1000
 soundClassObject=&Fort_Loot_Collect
 volumeMultiplier=0.8
 attenuationOverrides {falloffDistance=1000, attenuationShapeExtents {x=0}}
 concurrencyOverrides {maxCount=3, resolutionRule=PreventNew}
}
