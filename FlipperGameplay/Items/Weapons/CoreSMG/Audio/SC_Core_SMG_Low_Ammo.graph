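# Sound cue for the Core SMG's low-ammo warning sound.
# A single wave player feeds the output directly, reusing a shared high rifle low-ammo sample pitched up by the cue's multiplier.
# It uses close weapon-projectile attenuation and a ranged-weapon low-ammo sound class, and overrides concurrency to prevent new instances once the limit is reached.
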SoundCue /FlipperGameplay/Items/Weapons/CoreSMG/Audio/SC_Core_SMG_Low_Ammo
root=WavePlayer_1
nodes [{id=WavePlayer_1, properties {soundWaveAssetPtr=/Game/Sounds/Fort_Weapons/LowAmmo/weapon_low_ammo_hi_rifle_05}}]
properties {
 attenuationSettings=&Fort_Weapon_Projectile_Close_Athena_Att
 bOverrideConcurrency=true
 maxDistance=4950
 pitchMultiplier=1.2
 soundClassObject=/Game/Sounds/FortSoundClassesAndModes/Fort_WeaponRanged_LowAmmo_Degraded_PSM
 volumeMultiplier=1
 concurrencyOverrides {resolutionRule=PreventNew}
}
