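# Sound cue for picking up the Core SMG weapon, classed as loot-collect audio.
# A random node picks between two equally weighted pickup waves, feeding a root Modulator left at default settings.
# It uses the shared weapon-pickup attenuation and the PickupMaximum_Oldest concurrency set, with reduced volume.
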SoundCue /FlipperGameplay/Items/Weapons/CoreSMG/Audio/SC_Core_SMG_Pickup
root=Modulator_0
nodes [
 {id=Modulator_0, children [Random_1]}
 {id=Random_1, children [WavePlayer_1,WavePlayer_3], properties {weights [1,1]}}
 {id=WavePlayer_1, properties {soundWaveAssetPtr=&/Core_SMG_PickupNew_01}}
 {id=WavePlayer_3, properties {soundWaveAssetPtr=&/Core_SMG_PickupNew_02}}
]
properties {
 attenuationSettings=&Fort_Weapon_Pickup_Att
 maxDistance=3500
 soundClassObject=&Fort_Loot_Collect
 volumeMultiplier=0.5077
 concurrencyOverrides {maxCount=1, resolutionRule=PreventNew}
 concurrencySet [&PickupMaximum_Oldest]
}
