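# Sound cue for the cartridge eject foley during the Core SMG reload.
# A Random node picks one of three equally weighted eject waves and passes it through a Modulator at default settings for slight variation.
# The cue uses weapon reload attenuation and the weapon foley sound class at reduced volume, overrides concurrency, and has its debug flag enabled.
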SoundCue /FlipperGameplay/Items/Weapons/CoreSMG/Audio/SC_Core_SMG_Reload_Cartridge_Eject
root=Modulator_2
nodes [
 {id=Modulator_2, children [Random_0]}
 {id=Random_0, children [WavePlayer_0,WavePlayer_1,WavePlayer_6], properties {weights [1,1,1]}}
 {id=WavePlayer_0, properties {soundWaveAssetPtr=&/Core_SMG_ReloadEjectNew_03}}
 {id=WavePlayer_1, properties {soundWaveAssetPtr=&/Core_SMG_ReloadEjectNew_02}}
 {id=WavePlayer_6, properties {soundWaveAssetPtr=&/Core_SMG_ReloadEjectNew_01}}
]
properties {
 attenuationSettings=&Fort_Weapon_Reload_Att
 bDebug=true
 bOverrideConcurrency=true
 maxDistance=2250
 soundClassObject=&Fort_WeaponFoley
 volumeMultiplier=0.451
}
