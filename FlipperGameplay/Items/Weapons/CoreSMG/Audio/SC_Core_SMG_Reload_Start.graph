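# Sound cue for the start of the Core SMG's reload foley.
# A Random node picks one of three equally weighted reload-start wave variations, then a Modulator node with default settings passes it to the output.
# It plays in the weapon foley sound class at about half volume with shared reload attenuation, overrides concurrency, and has its debug flag enabled.
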SoundCue /FlipperGameplay/Items/Weapons/CoreSMG/Audio/SC_Core_SMG_Reload_Start
root=Modulator_0
nodes [
 {id=Modulator_0, children [Random_1]}
 {id=Random_1, children [WavePlayer_2,WavePlayer_3,WavePlayer_4], properties {weights [1,1,1]}}
 {id=WavePlayer_2, properties {soundWaveAssetPtr=&/Core_SMG_ReloadStartNew_01}}
 {id=WavePlayer_3, properties {soundWaveAssetPtr=&/Core_SMG_ReloadStartNew_02}}
 {id=WavePlayer_4, properties {soundWaveAssetPtr=&/Core_SMG_ReloadStartNew_03}}
]
properties {
 attenuationSettings=&Fort_Weapon_Reload_Att
 bDebug=true
 bOverrideConcurrency=true
 maxDistance=2250
 soundClassObject=&Fort_WeaponFoley
 volumeMultiplier=0.501
}
